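# Sound cue that plays when a Creative rift portal is destroyed, judging by its name.
# A single wave player feeds the root directly with the RiftPortal_Teleport_Collapse_07 wave, with no randomization or modulation.
# It uses the shared ability status-effect activation attenuation and the gadgets/items sound class.
# A concurrency override caps simultaneous instances and stops the oldest when the limit is reached.
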SoundCue /Game/Sounds/Fort_GamePlay_Sounds/Player/RiftPortal/RiftPortal_Creative_Destroy_Cue
root=WavePlayer_1
nodes [{id=WavePlayer_1, properties {soundWaveAssetPtr=&/RiftPortal_Teleport_Collapse_07}}]
properties {
 attenuationSettings=/Game/Sounds/Attenuation/Fort_Ability_Status_Effect_Activation_Att
 bOverrideConcurrency=true
 maxDistance=6000
 soundClassObject=&Fort_Gadgets_Items
 concurrencyOverrides {maxCount=8, resolutionRule=StopOldest}
}
