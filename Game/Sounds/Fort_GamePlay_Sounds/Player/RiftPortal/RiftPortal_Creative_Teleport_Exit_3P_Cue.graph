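# Sound cue for the third-person exit sound of a Creative Rift Portal teleport.
# A Random node picks among five equally weighted exit waves, feeding a Modulator that holds volume fixed and leaves pitch at its default range.
# The cue boosts volume, uses the distant binaural Rift Portal attenuation asset, overrides concurrency to cap simultaneous instances, and routes to the critical in-game UI sound class.
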SoundCue /Game/Sounds/Fort_GamePlay_Sounds/Player/RiftPortal/RiftPortal_Creative_Teleport_Exit_3P_Cue
root=Modulator_0
nodes [
 {id=Modulator_0, children [Random_0], properties {volumeMax=1, volumeMin=1}}
 {
  id=Random_0
  children [WavePlayer_0,WavePlayer_2,WavePlayer_3,WavePlayer_4,WavePlayer_5]
  properties {bShouldExcludeFromBranchCulling=true, weights [1,1,1,1,1]}
 }
 {id=WavePlayer_0, properties {soundWaveAssetPtr=&/RiftPortal_Teleport_Exit_3P_04}}
 {id=WavePlayer_2, properties {soundWaveAssetPtr=&/RiftPortal_Teleport_Exit_3P_05}}
 {id=WavePlayer_3, properties {soundWaveAssetPtr=&/RiftPortal_Teleport_Exit_3P_06}}
 {id=WavePlayer_4, properties {soundWaveAssetPtr=&/RiftPortal_Teleport_Exit_3P_01}}
 {id=WavePlayer_5, properties {soundWaveAssetPtr=&/RiftPortal_Teleport_Exit_3P_02}}
]
properties {
 attenuationSettings=&/RiftPortal_Distant_Binaural_Att
 bOverrideConcurrency=true
 maxDistance=60000
 soundClassObject=&Fort_UI_InGame_Critical
 volumeMultiplier=3
 attenuationOverrides {bAttenuate=false}
 concurrencyOverrides {maxCount=4}
}
