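# Sound cue for the Creative rift portal's teleport relocate sound.
# A single wave player feeds the relocate wave straight to the output at reduced volume, routed to the non-critical in-game UI sound class.
# It references the distant rift portal attenuation asset for long-range audibility.
# A concurrency override allows only one instance at a time and blocks new plays while it is active.
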SoundCue /Game/Sounds/Fort_GamePlay_Sounds/Player/RiftPortal/RiftPortal_Creative_Teleport_Relocate_Cue
root=WavePlayer_0
nodes [{id=WavePlayer_0, properties {soundWaveAssetPtr=&/Creative_Rift_Teleport_Relocate}}]
properties {
 attenuationSettings=&/RiftPortal_Distant_Att
 bOverrideConcurrency=true
 maxDistance=60000
 soundClassObject=&Fort_UI_InGame_NonCritical
 volumeMultiplier=0.7
 attenuationOverrides {bAttenuate=false}
 concurrencyOverrides {maxCount=1, resolutionRule=PreventNew}
}
