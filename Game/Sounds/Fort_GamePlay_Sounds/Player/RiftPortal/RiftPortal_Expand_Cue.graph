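# RiftPortal_Expand_Cue is a SoundCue for the expand sound of a rift portal, filed under player gameplay sounds.
# A Random node picks between two equally weighted teleport-expand waves.
# Playback uses the shared forged-consumable in-world attenuation and the gadgets and items sound class.
# A concurrency override caps simultaneous instances and stops the oldest when the limit is reached.
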SoundCue /Game/Sounds/Fort_GamePlay_Sounds/Player/RiftPortal/RiftPortal_Expand_Cue
root=Random_0
nodes [
 {id=Random_0, children [WavePlayer_3,WavePlayer_4], properties {weights [1,1]}}
 {id=WavePlayer_3, properties {soundWaveAssetPtr=&/RiftPortal_Teleport_Expand_06}}
 {id=WavePlayer_4, properties {soundWaveAssetPtr=&/RiftPortal_Teleport_Expand_08}}
]
properties {
 attenuationSettings=/Game/Sounds/Attenuation/Fort_ForgedConsumable_InWorld_Att
 bOverrideConcurrency=true
 maxDistance=1000
 soundClassObject=&Fort_Gadgets_Items
 concurrencyOverrides {maxCount=8, resolutionRule=StopOldest}
}
