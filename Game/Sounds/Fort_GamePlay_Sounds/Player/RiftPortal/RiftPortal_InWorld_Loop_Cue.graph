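# Looping sound cue for a rift portal's in-world ambience, classed as a gadget/item sound.
# A quality switch gives the first two levels a mix of a main loop and a slightly louder close-up loop with its own overridden attenuation.
# The last two levels play only an "_opt" main loop, using the same shared rift portal attenuation as the full main loop.
# A concurrency override caps instances, stopping the quietest.
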SoundCue /Game/Sounds/Fort_GamePlay_Sounds/Player/RiftPortal/RiftPortal_InWorld_Loop_Cue
root=QualityLevel_0
nodes [
 {
  id=Attenuation_0
  children [WavePlayer_1]
  properties {bOverrideAttenuation=true, attenuationOverrides {falloffDistance=512, attenuationShapeExtents {x=0}}}
 }
 {
  id=Attenuation_2
  children [WavePlayer_0]
  properties {attenuationSettings=/Game/Sounds/Attenuation/Fort_RiftPortals_Att, attenuationOverrides {falloffDistance=60, attenuationShapeExtents {x=0}}}
 }
 {
  id=Attenuation_3
  children [WavePlayer_2]
  properties {attenuationSettings=&Fort_RiftPortals_Att, attenuationOverrides {falloffDistance=60, attenuationShapeExtents {x=0}}}
 }
 {id=Mixer_0, children [Attenuation_2, Attenuation_0], properties {inputVolume [1,1.25]}}
 {id=QualityLevel_0, children [Mixer_0, Mixer_0, Attenuation_3, Attenuation_3]}
 {id=WavePlayer_0, properties {bLooping=true, soundWaveAssetPtr=/Game/Athena/Sounds/Events/Geode/Oddities/Geode_Oddities_InWorldLoop_01}}
 {
  id=WavePlayer_1
  properties {bLooping=true, soundWaveAssetPtr=&/RiftPortal_InWorldLoop_Close_01}
 }
 {id=WavePlayer_2, properties {bLooping=true, soundWaveAssetPtr=/Game/Athena/Sounds/Events/Geode/Oddities/Geode_Oddities_InWorldLoop_01_opt}}
]
properties {
 bOverrideConcurrency=true
 maxDistance=4000
 soundClassObject=&Fort_Gadgets_Items
 concurrencyOverrides {maxCount=8, resolutionRule=StopQuietest}
}
