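# Sound cue for the rift portal shard consume sound, filed under player gameplay sounds.
# A Quality Level node at the root plays the standard wave in its first two slots and the "_opt" variant in the last two.
# It routes to the Gadgets Items sound class and uses the shared ability status effect activation attenuation.
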SoundCue /Game/Sounds/Fort_GamePlay_Sounds/Player/RiftPortal/RiftPortal_ShardConsume_Cue
root=QualityLevel_0
nodes [
 {id=QualityLevel_0, children [WavePlayer_0,WavePlayer_0,WavePlayer_1,WavePlayer_1]}
 {id=WavePlayer_0, properties {soundWaveAssetPtr=&/RiftPortal_ShardConsume_01}}
 {id=WavePlayer_1, properties {soundWaveAssetPtr=&/RiftPortal_ShardConsume_01_opt}}
]
properties {
 attenuationSettings=/Game/Sounds/Attenuation/Fort_Ability_Status_Effect_Activation_Att
 maxDistance=6000
 soundClassObject=&Fort_Gadgets_Items
 volumeMultiplier=1
}
