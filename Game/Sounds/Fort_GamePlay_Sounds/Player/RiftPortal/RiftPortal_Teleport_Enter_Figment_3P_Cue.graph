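# Sound cue for the Figment variant of the rift portal teleport enter sound, apparently as heard in third person from other players.
# A single wave player feeds a Modulator node at the root, playing one rift portal enter wave.
# The cue uses Figment-specific rift portal attenuation, the gadgets and items sound class, and a lowered volume multiplier.
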SoundCue /Game/Sounds/Fort_GamePlay_Sounds/Player/RiftPortal/RiftPortal_Teleport_Enter_Figment_3P_Cue
root=Modulator_0
nodes [
 {id=Modulator_0, children [WavePlayer_0]}
 {id=WavePlayer_0, properties {soundWaveAssetPtr=&/RiftPortal_Teleport_Enter_3P_02}}
]
properties {
 attenuationSettings=/Game/Sounds/Attenuation/Fort_RiftPortals_Figment_Att
 maxDistance=8000
 soundClassObject=&Fort_Gadgets_Items
 volumeMultiplier=0.6
}
