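# Sound cue for the first-person rift portal teleport exit sound, in the gadgets and items sound class.
# A quality-level node feeds the surround teleport wave to its first two slots and the "opt" teleport wave to its last two.
# Concurrency is overridden to allow a single instance per owner.
# A very large max distance is set on the cue.
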SoundCue /Game/Sounds/Fort_GamePlay_Sounds/Player/RiftPortal/RiftPortal_Teleport_Exit_1P_Cue
root=QualityLevel_0
nodes [
 {id=QualityLevel_0, children [WavePlayer_2,WavePlayer_2,WavePlayer_1,WavePlayer_1]}
 {id=WavePlayer_1, properties {soundWaveAssetPtr=&/RiftPortal_Teleport_opt}}
 {id=WavePlayer_2, properties {soundWaveAssetPtr=&/RiftPortal_Teleport_Surround_01}}
]
properties {
 bOverrideConcurrency=true
 maxDistance=2097152
 soundClassObject=&Fort_Gadgets_Items
 volumeMultiplier=1
 concurrencyOverrides {bLimitToOwner=true, maxCount=1}
}
