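# Sound cue for the quick first-person exit sound of a Rift Portal teleport, routed through the Fort_Gadgets_Items sound class.
# A single wave player feeds the Quick_01 surround teleport wave through a Modulator left at default settings, with a boosted cue volume.
# A concurrency override limits playback to one instance per owner.
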SoundCue /Game/Sounds/Fort_GamePlay_Sounds/Player/RiftPortal/RiftPortal_Teleport_Exit_1P_Quick_Cue
root=Modulator_0
nodes [
 {id=Modulator_0, children [WavePlayer_3]}
 {id=WavePlayer_3, properties {soundWaveAssetPtr=&/RiftPortal_Teleport_Surround_Quick_01}}
]
properties {
 bOverrideConcurrency=true
 maxDistance=2097152
 soundClassObject=&Fort_Gadgets_Items
 volumeMultiplier=1.25
 concurrencyOverrides {bLimitToOwner=true, maxCount=1}
}
