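# Sound cue for the third-person rift portal teleport exit, in the gadgets and items sound class with limited concurrency.
# A mixer blends a main layer of five random variants with a quieter focus-detail layer of six random variants.
# Each layer has its own distant binaural attenuation node.
# A quality-level node plays the detail layer only in its first two slots, leaving the other two empty.
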SoundCue /Game/Sounds/Fort_GamePlay_Sounds/Player/RiftPortal/RiftPortal_Teleport_Exit_3P_Cue
root=Mixer_0
nodes [
 {
  id=Attenuation_0
  children [Modulator_1]
  properties {attenuationSettings=&/RiftPortal_Distant_Binaural_Att}
 }
 {
  id=Attenuation_1
  children [QualityLevel_0]
  properties {attenuationSettings=&/RiftPortal_Distant_Binaural_Focus_Detail_Att}
 }
 {id=Mixer_0, children [Attenuation_0, Attenuation_1], properties {inputVolume [1,0.75]}}
 {id=Modulator_1, children [Random_0], properties {volumeMax=1, volumeMin=1}}
 {id=QualityLevel_0, children [Random_1,Random_1,null,null]}
 {
  id=Random_0
  children [WavePlayer_0,WavePlayer_1,WavePlayer_3,WavePlayer_4,WavePlayer_5]
  properties {bShouldExcludeFromBranchCulling=true, weights [1,1,1,1,1]}
 }
 {
  id=Random_1
  children [WavePlayer_2, WavePlayer_6, WavePlayer_7, WavePlayer_8, WavePlayer_9, WavePlayer_10]
  properties {bShouldExcludeFromBranchCulling=true, weights [1,1,1,1,1,1]}
 }
 {id=WavePlayer_0, properties {soundWaveAssetPtr=&/RiftPortal_Teleport_Exit_3P_04}}
 {id=WavePlayer_1, properties {soundWaveAssetPtr=&/RiftPortal_Teleport_Exit_3P_05}}
 {
  id=WavePlayer_10
  properties {soundWaveAssetPtr=&/RiftPortal_Teleport_Exit_3P_Focus_Detail_06}
 }
 {
  id=WavePlayer_2
  properties {soundWaveAssetPtr=&/RiftPortal_Teleport_Exit_3P_Focus_Detail_01}
 }
 {id=WavePlayer_3, properties {soundWaveAssetPtr=&/RiftPortal_Teleport_Exit_3P_06}}
 {id=WavePlayer_4, properties {soundWaveAssetPtr=&/RiftPortal_Teleport_Exit_3P_01}}
 {id=WavePlayer_5, properties {soundWaveAssetPtr=&/RiftPortal_Teleport_Exit_3P_02}}
 {
  id=WavePlayer_6
  properties {soundWaveAssetPtr=&/RiftPortal_Teleport_Exit_3P_Focus_Detail_02}
 }
 {
  id=WavePlayer_7
  properties {soundWaveAssetPtr=&/RiftPortal_Teleport_Exit_3P_Focus_Detail_03}
 }
 {
  id=WavePlayer_8
  properties {soundWaveAssetPtr=&/RiftPortal_Teleport_Exit_3P_Focus_Detail_04}
 }
 {
  id=WavePlayer_9
  properties {soundWaveAssetPtr=&/RiftPortal_Teleport_Exit_3P_Focus_Detail_05}
 }
]
properties {
 bOverrideConcurrency=true
 maxDistance=60000
 soundClassObject=&Fort_Gadgets_Items
 volumeMultiplier=0.55
 attenuationOverrides {bAttenuate=false}
 concurrencyOverrides {maxCount=4}
}
